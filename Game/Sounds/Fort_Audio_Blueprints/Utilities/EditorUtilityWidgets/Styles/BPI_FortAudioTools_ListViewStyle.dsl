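; Blueprint interface for list view styling in the Fortnite audio editor utility widget tools.
; It declares two functions, GetIsSelected and GetRowIndex, which expose whether an entry is selected and its row position.
; The functions carry only placeholder returns of false and zero, so implementers supply the actual values.
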(fn GetIsSelected
  (return false))

(fn GetRowIndex
  (return 0))
