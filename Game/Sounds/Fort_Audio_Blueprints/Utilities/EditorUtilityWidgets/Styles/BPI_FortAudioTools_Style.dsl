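; Blueprint interface for the styling of Fortnite audio editor utility widgets, kept with the audio tools' Styles assets.
; It declares one function, GetMetaSoundDataTypeColor, which takes a MetaSound data type name and, by its name, supplies a color for that type.
; The function holds no logic here and returns only a default value, leaving the actual color lookup to implementers.
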(fn GetMetaSoundDataTypeColor
  (params (Data%20Type :type Name))
  (return 0))
