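# Editor utility function library that styles list rows in the Fort Audio Tools editor widgets.
# Both functions cast the list entry widget to the list view style interface to query selection, and also check hover.
# Background color layers selected blue over a hover shade over alternating stripes from row index parity.
# Text color switches from grey to white when the row is selected or hovered.
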EditorUtilityBlueprint /Game/Sounds/Fort_Audio_Blueprints/Utilities/EditorUtilityWidgets/Styles/EFL_FortAudioTools_Style
graphs [
 Get List Row Background Color {nodes [
   n1 = Get List Row Background Color!() -> (then, List Entry Widget:Editor Utility Widget Object Reference) {is=title K2Node_FunctionEntry}
   > n2 = Cast To BPI_FortAudioTools_ListViewStyle!(execute, Object=@n1) -> (then, CastFailed, AsFortAudioTools List View Interface:Fort Audio Tools List View Interface Interface) {is=title K2Node_DynamicCast}
   > n3 = Get Row Index!(execute, self:Fort Audio Tools List View Interface Interface=@n2) -> (then, Row Index:int=0) {is=title K2Node_CallFunction}
   > n4 = Get Is Selected!(execute, self:Fort Audio Tools List View Interface Interface=@n2) -> (then, Is Selected:bool) {is=title K2Node_CallFunction}
   n5 = IsHovered(self=Get List Entry Widget() -> (List Entry Widget:Editor Utility Widget Object Reference))
   n6 = Select(
    Option 0:Linear Color Structure=(R=0.0103,G=0.0103,B=0.0103,A=1)
    Option 1:Linear Color Structure=(R=0.0075,G=0.0075,B=0.0075,A=1)
    Index="Equal (Integer)"(B="% (Integer)"(A=@n3, B=2) -> (ReturnValue=0))
   ) -> (ReturnValue:Linear Color Structure)
   > Return Node!(
    execute
    ReturnValue:Linear Color Structure=Select(Option 0=Select(Option 0=@n6, Option 1=(R=0.0176,G=0.0176,B=0.0176,A=1), Index=@n5), Option 1=(R=0,G=0.162,B=0.7454,A=1), Index=@n4)
   )
  ]}
 Get List Row Text Color {nodes [
   n1 = Get List Row Text Color!() -> (then, List Entry Widget:Editor Utility Widget Object Reference) {is=title K2Node_FunctionEntry}
   > n2 = Cast To BPI_FortAudioTools_ListViewStyle(Object=@n1)
   > n3 = Get Is Selected(self=@n2)
   n4 = OR Boolean(A=@n3, B=IsHovered(self=@n1))
   > Return Node(ReturnValue:Slate Color Structure=Make Slate Color(SpecifiedColor=Select(Option 0=(R=0.5029,G=0.5029,B=0.5029,A=1), Option 1=(R=1,G=1,B=1,A=1), Index=@n4)))
  ]}
]
variables []
