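; FortAudioTools_Style supplies shared color functions for Fortnite audio editor utility widgets.
; List row functions query each entry through the list view style interface, coloring by selection and hover, with backgrounds otherwise alternating by row index parity.
; A third function maps MetaSound data type names to colors through a local map and enum select.
; Unmapped names get a dedicated color if prefixed "Enum:", otherwise a default.
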(fn GetListRowBackgroundColor
  (params (List%20Entry%20Widget :type Editor%20Utility%20Widget%20Object%20Reference))
  (bind _asfortaudiotools_list_view_interface (Utilities|Casting|CastToBPI_FortAudioTools_ListViewStyle List%20Entry%20Widget))
  (bind _row_index (Class|FortAudioToolsListViewInterface|GetRowIndex _asfortaudiotools_list_view_interface))
  (bind _is_selected (Class|FortAudioToolsListViewInterface|GetIsSelected _asfortaudiotools_list_view_interface))
  (return (select _is_selected "(R=0.000000,G=0.162029,B=0.745404,A=1.000000)" (select (Widget|IsHovered (Variables|GetList%20Entry%20Widget)) "(R=0.017642,G=0.017642,B=0.017642,A=1.000000)" (select (== 0 (Math|Integer|%25%28Integer%29 _row_index 2)) "(R=0.007499,G=0.007499,B=0.007499,A=1.000000)" "(R=0.010330,G=0.010330,B=0.010330,A=1.000000)")))))

(fn GetListRowTextColor
  (params (List%20Entry%20Widget :type Editor%20Utility%20Widget%20Object%20Reference))
  ; To be refactored
  (bind _asfortaudiotools_list_view_interface (Utilities|Casting|CastToBPI_FortAudioTools_ListViewStyle List%20Entry%20Widget))
  (bind _is_selected (Class|FortAudioToolsListViewInterface|GetIsSelected _asfortaudiotools_list_view_interface))
  (return (Utilities|Struct|MakeSlateColor (select (or _is_selected (Widget|IsHovered List%20Entry%20Widget)) "(R=1.000000,G=1.000000,B=1.000000,A=1.000000)" "(R=0.502887,G=0.502887,B=0.502887,A=1.000000)"))))

(fn GetMetaSoundDataTypeColor
  (params (Data%20Type :type Name :default "None"))
  (bind (_value _returnvalue) (Utilities|Map|Find (LocalVariables|GetData%20Type%20Mappings) (Variables|GetData%20Type)))
  (return (select _returnvalue (Utilities|Select :Index _value :NewEnumerator0 "(R=1.000000,G=1.000000,B=1.000000,A=1.000000)" :NewEnumerator1 "(R=0.300544,G=0.000000,B=0.000000,A=1.000000)" :NewEnumerator9 "(R=0.300544,G=0.000000,B=0.000000,A=1.000000)" :NewEnumerator2 "(R=0.356400,G=1.000000,B=0.059511,A=1.000000)" :NewEnumerator10 "(R=0.356400,G=1.000000,B=0.059511,A=1.000000)" :NewEnumerator3 "(R=0.013702,G=0.768151,B=0.428690,A=1.000000)" :NewEnumerator11 "(R=0.013702,G=0.768151,B=0.428690,A=1.000000)" :NewEnumerator4 "(R=0.300544,G=1.000000,B=1.000000,A=1.000000)" :NewEnumerator12 "(R=0.300544,G=1.000000,B=1.000000,A=1.000000)" :NewEnumerator7 "(R=0.000000,G=0.401978,B=0.913099,A=1.000000)" :NewEnumerator8 "(R=0.000000,G=0.401978,B=0.913099,A=1.000000)") (select (Utilities|String|StartsWith (Utilities|String|ToString(Name) (Variables|GetData%20Type)) "Enum:") "(R=0.000000,G=0.152926,B=0.127438,A=1.000000)" (LocalVariables|GetDefault%20Color)))))
